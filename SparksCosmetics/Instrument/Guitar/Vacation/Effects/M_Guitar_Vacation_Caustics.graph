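# Additive, unlit, two-sided Niagara sprite material for animated caustics effects on the Vacation guitar cosmetic.
# Sums the red channels of two samples from caustics textures, panned in different directions with one tiled slightly tighter, then multiplies by particle color.
# A square-rooted exponential radial gradient times particle alpha forms a soft circular mask.
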Material /SparksCosmetics/Instrument/Guitar/Vacation/Effects/M_Guitar_Vacation_Caustics
graph {
 properties {
  bAutomaticallySetUsageInEditor=false
  bCastRayTracedShadows=false
  bUseTranslucencyVertexFog=false
  bUsedWithNiagaraSprites=true
  blendMode=BLEND_Additive
  pixelDepthOffsetMode=PDOM_Legacy
  shadingModel=MSM_Unlit
  twoSided=true
  usedShadingModels=Unlit
 }
 nodes [
  n1 = ParticleColor()
  n2 = Panner(Coordinate=TextureCoordinate() {uTiling=1.3, vTiling=1.3}) {speedX=-0.05, speedY=-0.05}
  n3 = Panner() {speedX=0.025, speedY=-0.025}
  Multiply(
   A=@n1.RGB
   B=Add(
    A=TextureSample(UVs=@n2) {samplerType=SAMPLERTYPE_LinearColor, texture=/ItemShop/UI/Art/WaterBackground/T_UI_Caustics}.R
    B=TextureSample(UVs=@n3) {texture=&T_ShieldCaustics}.R
   )
  )
  Multiply(
   A=@n1.A
   B=Power(Base=MaterialFunctionCall() {materialFunction=&RadialGradientExponential}.RadialGradientExponential) {constExponent=0.5}
  )
 ]
}
