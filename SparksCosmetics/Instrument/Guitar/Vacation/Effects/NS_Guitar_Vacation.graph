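# Niagara system for the Vacation guitar cosmetic, adding underwater-style bubble and caustic sprites.
# Bubbles_Main_New and Caustics_New partition particles across array-driven offsets, fade them by curve and jitter positions with randomized lerps.
# User.bIsFrontEndPreview switches caustic color and alpha strength for front-end preview.
# Older per-corner bubble and caustic emitters, plus Fish and StarFish, which follow skeletal mesh bones or sockets, are disabled.
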NiagaraSystem /SparksCosmetics/Instrument/Guitar/Vacation/Effects/NS_Guitar_Vacation
dependencies {
 usedDataInterfaces [
  &NiagaraDataInterfaceCurve
  &NiagaraDataInterfaceArrayInt32
  &NiagaraDataInterfaceArrayFloat3
  &NiagaraDataInterfaceSkeletalMesh
 ]
 usedDynamicInputs []
 usedModules [
  &self:SystemSpawnScript.SystemScriptSource.SystemScriptGraph.NiagaraNodeAssignment_0.SetVariables_#1
  &SystemState
 ]
}
emitters {
 Bubbles_Bottom_Left {
  bEnabled=false
  emitterUpdateScript [
   &EmitterState()
   &SpawnRate(Spawn Probability=0.5, SpawnRate=0.35, Use Spawn Probability=false)
  ]
  particleSpawnScript [
   &InitializeParticle(
    Color=<rgba 0.5578 0.839 1 1>
    Lifetime Max=4
    Lifetime Min=3
    Lifetime Mode=::NewEnumerator1
    Position Offset=<0 6 11>
    Sprite Rotation Mode=::NewEnumerator2
    Sprite Size Mode=::NewEnumerator1
    Sprite UV Mode=::NewEnumerator3
    Uniform Sprite Size Max=2
    Uniform Sprite Size Min=4
    UsePositionOffset=true
    Write Material Random=true
   )
   &PartitionParticles(Number Of Partitions=5)
   &ShapeLocation(
    Cylinder Height=0
    Cylinder Radius=5
    MaskSpawnGroup=false
    Override Seed=false
    Rotation Mode=::NewEnumerator1
    Shape Primitive=::NewEnumerator1
    Surface Only=false
    Transform Order:ENiagara_TransformOrder=::NewEnumerator1
    UseOwnerScale=true
   )
   &AddVelocity(Velocity=&RandomRangeVector(Minimum=<1 0 1>, Maximum=<-1 0 -1>))
  ]
  particleUpdateScript [
   &ParticleState()
   &ScaleSpriteSize(Scale Factor=&Vector2DFromFloat(Value=&FloatFromCurve(FloatCurve=DataInterface() {propertyValues {Curve {keys [
          {arriveTangent=9.0725, interpMode=RCIM_Cubic, leaveTangent=9.0725, tangentMode=RCTM_User, time=0, value=0}
          {arriveTangent=0.001, interpMode=RCIM_Cubic, leaveTangent=0.001, tangentMode=RCTM_User, time=0.2, value=0.95}
          {arriveTangent=-1.044, interpMode=RCIM_Cubic, leaveTangent=-1.044, tangentMode=RCTM_User, time=0.6, value=0.85}
          {arriveTangent=-2.0004, interpMode=RCIM_Cubic, leaveTangent=-2.0004, tangentMode=RCTM_User, time=1, value=0}
         ]}}})))
   &ScaleColor(
    Scale Alpha=&FloatFromCurve(FloatCurve=DataInterface() {propertyValues {Curve {keys [<at 0 0 RCIM_Cubic>, <at 0.2 1>, {arriveTangent=-1.25, interpMode=RCIM_Cubic, leaveTangent=-1.25, time=0.7, value=1}, <at 1 0 RCIM_Cubic>]}}})
    ScaleA=true
    ScaleRGB=true
   )
   &Drag(Drag=0.5, Use Linear Drag=true, Use Rotational Drag=true) {enabled=false}
   &VortexForce(UseOriginOffset=false, Vortex Axis=<0 1 0>, Vortex Force Amount=1) {enabled=false}
   &CurlNoiseForce(Noise Frequency=20, Noise Strength=3, Pan Noise=false) {enabled=false}
   &SolveForcesAndVelocity(Clamp Velocity=false, Limit Acceleration=false)
   &SkeletalMeshLocation(
    Bone / Socket Index=&SelectIntFromArray(Direct Array Index=$Particles.PartitionParticles.Partition, Int Selection Array=DataInterface() {propertyValues {IntData [3]}})
    Bone / Socket Sampling Mode=::NewEnumerator5
    MaskSpawnGroup=false
    Mesh Sampling Type=::NewEnumerator6
    Position Sampling=::NewEnumerator1
    Sampled Position Offset=$Particles.ShapeLocation.ShapePosition
    Skeletal Mesh=$System.Skeletal Mesh
   )
  ]
  rendererClasses [&NiagaraSpriteRendererProperties]
  renderers [{rendererClass=&NiagaraSpriteRendererProperties}]
 }
 Bubbles_Main {
  bEnabled=false
  emitterUpdateScript [
   &EmitterState()
   &SpawnRate(Spawn Probability=0.5, SpawnRate=0.35, Use Spawn Probability=false)
  ]
  particleSpawnScript [
   &InitializeParticle(
    Color=<rgba 0.5578 0.839 1 1>
    Lifetime Max=5
    Lifetime Min=3
    Lifetime Mode=::NewEnumerator1
    Position Offset=<0 0 20>
    Sprite Rotation Mode=::NewEnumerator2
    Sprite Size Mode=::NewEnumerator1
    Sprite UV Mode=::NewEnumerator3
    Uniform Sprite Size Max=2
    UsePositionOffset=true
    Write Material Random=true
   )
   &PartitionParticles(Number Of Partitions=5)
   &ShapeLocation(
    Box Size=<25 11 0>
    MaskSpawnGroup=false
    Override Seed=false
    Rotation Mode=::NewEnumerator1
    Shape Primitive=::NewEnumerator2
    Surface Only=false
    Transform Order:ENiagara_TransformOrder=::NewEnumerator1
    UseOwnerScale=true
   )
   &AddVelocity(Velocity=&RandomRangeVector(Minimum=<1.5 0 1.5>, Maximum=<-1.5 0 -1.5>))
  ]
  particleUpdateScript [
   &ParticleState()
   &ScaleSpriteSize(Scale Factor=&Vector2DFromFloat(Value=&FloatFromCurve(FloatCurve=DataInterface() {propertyValues {Curve {keys [
          {arriveTangent=9.0725, interpMode=RCIM_Cubic, leaveTangent=9.0725, tangentMode=RCTM_User, time=0, value=0}
          {arriveTangent=0.001, interpMode=RCIM_Cubic, leaveTangent=0.001, tangentMode=RCTM_User, time=0.2, value=0.95}
          {arriveTangent=-1.044, interpMode=RCIM_Cubic, leaveTangent=-1.044, tangentMode=RCTM_User, time=0.6, value=0.85}
          {arriveTangent=-2.0004, interpMode=RCIM_Cubic, leaveTangent=-2.0004, tangentMode=RCTM_User, time=1, value=0}
         ]}}})))
   &ScaleColor(
    Scale Alpha=&FloatFromCurve(FloatCurve=DataInterface() {propertyValues {Curve {keys [<at 0 0 RCIM_Cubic>, <at 0.2 1>, {arriveTangent=-1.25, interpMode=RCIM_Cubic, leaveTangent=-1.25, time=0.7, value=1}, <at 1 0 RCIM_Cubic>]}}})
    ScaleA=true
    ScaleRGB=true
   )
   &Drag(Drag=0.5, Use Linear Drag=true, Use Rotational Drag=true) {enabled=false}
   &VortexForce(Origin Pull Amount=1, UseOriginOffset=false, Vortex Axis=<0 1 0>, Vortex Force Amount=1) {enabled=false}
   &CurlNoiseForce(Noise Frequency=20, Noise Strength=3, Pan Noise=false) {enabled=false}
   &SolveForcesAndVelocity(Clamp Velocity=false, Limit Acceleration=false)
   &SkeletalMeshLocation(
    Bone / Socket Index=&SelectIntFromArray(Direct Array Index=$Particles.PartitionParticles.Partition, Int Selection Array=DataInterface() {propertyValues {IntData [3]}})
    Bone / Socket Sampling Mode=::NewEnumerator5
    MaskSpawnGroup=false
    Mesh Sampling Type=::NewEnumerator6
    Position Sampling=::NewEnumerator1
    Sampled Position Offset=$Particles.ShapeLocation.ShapePosition
    Skeletal Mesh=$System.Skeletal Mesh
   )
  ]
  rendererClasses [&NiagaraSpriteRendererProperties]
  renderers [{rendererClass=&NiagaraSpriteRendererProperties}]
 }
 Bubbles_Main_New {
  emitterUpdateScript [
   &EmitterState()
   &SpawnRate(Spawn Probability=0.75, SpawnRate=3, Use Spawn Probability=true)
  ]
  particleSpawnScript [
   &PartitionParticles(Number Of Partitions=4)
   &InitializeParticle(
    Color=<rgba 0.5578 0.839 1 1>
    Lifetime Max=4
    Lifetime Min=3
    Lifetime Mode=::NewEnumerator1
    Position Offset=&SelectVectorFromArray(
     Direct Array Index=$Particles.PartitionParticles.Partition
     Vector Selection Array=DataInterface() {propertyValues {FloatData [<0 -2 0>, <10 -2 10>, <-7 -2 12>, <-6 -2 -4>]}}
    )
    Sprite Rotation Mode=::NewEnumerator2
    Sprite Size Mode=::NewEnumerator1
    Sprite UV Mode=::NewEnumerator3
    Uniform Sprite Size Max=2
    UsePositionOffset=true
    Write Material Random=true
   )
   &ShapeLocation(
    Box Size=&SelectVectorFromArray(
     Direct Array Index=$Particles.PartitionParticles.Partition
     Vector Selection Array=DataInterface() {propertyValues {FloatData [<25 0 18>, <4.5 0 5>, <10 0 8>, <8 0 8>]}}
    )
    MaskSpawnGroup=false
    Offset Mode=::NewEnumerator0
    Override Seed=false
    Rotation Mode=::NewEnumerator1
    Shape Origin=$Particles.Position
    Shape Primitive=::NewEnumerator2
    Surface Only=false
    Transform Order:ENiagara_TransformOrder=::NewEnumerator1
    UseOwnerScale=true
   )
  ]
  particleUpdateScript [
   &ParticleState()
   &ScaleSpriteSize(Scale Factor=&Vector2DFromFloat(Value=&FloatFromCurve(FloatCurve=DataInterface() {propertyValues {Curve {keys [
          {arriveTangent=9.0725, interpMode=RCIM_Cubic, leaveTangent=9.0725, tangentMode=RCTM_User, time=0, value=0}
          {arriveTangent=0.001, interpMode=RCIM_Cubic, leaveTangent=0.001, tangentMode=RCTM_User, time=0.2, value=0.95}
          {arriveTangent=-1.044, interpMode=RCIM_Cubic, leaveTangent=-1.044, tangentMode=RCTM_User, time=0.6, value=0.85}
          {arriveTangent=-2.0004, interpMode=RCIM_Cubic, leaveTangent=-2.0004, tangentMode=RCTM_User, time=1, value=0}
         ]}}})))
   &ScaleColor(
    Scale Alpha=&FloatFromCurve(
     FloatCurve=DataInterface() {propertyValues {Curve {keys [<at 0 0 RCIM_Cubic>, <at 0.2 1>, {arriveTangent=-1.25, interpMode=RCIM_Cubic, leaveTangent=-1.25, time=0.7, value=1}, <at 1 0 RCIM_Cubic>]}}}
     Scale Curve=0.5
    )
    ScaleA=true
    ScaleRGB=true
   )
   /SetVariables(Particles.Position=&Lerp_Vector(
     A=Linked() {linkedVariable {name=Particles.Position, type=NiagaraPosition}}
     B=&AddVectorToPosition(
      Position=$Particles.Position
      Vector=&RandomRangeVector(Minimum=<-0.05 0 -0.05>, Maximum=<0.05 0 0.05>)
     )
     Alpha=&FloatFromCurve(FloatCurve=DataInterface() {propertyValues {Curve {keys [<at 0 0 RCIM_Cubic>, {tangentMode=RCTM_User, time=1, value=1}]}}})
    )) {isSetParameters=true}
  ]
  rendererClasses [&NiagaraSpriteRendererProperties]
  renderers [{rendererClass=&NiagaraSpriteRendererProperties}]
 }
 Bubbles_Upper_Left {
  bEnabled=false
  emitterUpdateScript [
   &EmitterState()
   &SpawnRate(SpawnRate=0.15, Use Spawn Probability=false)
  ]
  particleSpawnScript [
   &InitializeParticle(
    Color=<rgba 0.5578 0.839 1 1>
    Lifetime Max=4
    Lifetime Min=3
    Lifetime Mode=::NewEnumerator1
    Position Offset=<0 7 34>
    Sprite Rotation Mode=::NewEnumerator2
    Sprite Size Mode=::NewEnumerator1
    Sprite UV Mode=::NewEnumerator3
    Uniform Sprite Size Max=2
    UsePositionOffset=true
    Write Material Random=true
   )
   &PartitionParticles(Number Of Partitions=5)
   &ShapeLocation(
    Cylinder Height=0
    Cylinder Radius=4
    MaskSpawnGroup=false
    Override Seed=false
    Rotation Mode=::NewEnumerator1
    Shape Primitive=::NewEnumerator1
    Surface Only=false
    Transform Order:ENiagara_TransformOrder=::NewEnumerator1
    UseOwnerScale=true
   )
   &AddVelocity(Velocity=&RandomRangeVector(Minimum=<1 0 1>, Maximum=<-1 0 -1>))
  ]
  particleUpdateScript [
   &ParticleState()
   &ScaleSpriteSize(Scale Factor=&Vector2DFromFloat(Value=&FloatFromCurve(FloatCurve=DataInterface() {propertyValues {Curve {keys [
          {arriveTangent=9.0725, interpMode=RCIM_Cubic, leaveTangent=9.0725, tangentMode=RCTM_User, time=0, value=0}
          {arriveTangent=0.001, interpMode=RCIM_Cubic, leaveTangent=0.001, tangentMode=RCTM_User, time=0.2, value=0.95}
          {arriveTangent=-1.044, interpMode=RCIM_Cubic, leaveTangent=-1.044, tangentMode=RCTM_User, time=0.6, value=0.85}
          {arriveTangent=-2.0004, interpMode=RCIM_Cubic, leaveTangent=-2.0004, tangentMode=RCTM_User, time=1, value=0}
         ]}}})))
   &ScaleColor(
    Scale Alpha=&FloatFromCurve(FloatCurve=DataInterface() {propertyValues {Curve {keys [<at 0 0 RCIM_Cubic>, <at 0.2 1>, {arriveTangent=-1.25, interpMode=RCIM_Cubic, leaveTangent=-1.25, time=0.7, value=1}, <at 1 0 RCIM_Cubic>]}}})
    ScaleA=true
    ScaleRGB=true
   )
   &Drag(Drag=0.5, Use Linear Drag=true, Use Rotational Drag=true) {enabled=false}
   &VortexForce(UseOriginOffset=false, Vortex Axis=<0 1 0>, Vortex Force Amount=1) {enabled=false}
   &CurlNoiseForce(Noise Frequency=20, Noise Strength=3, Pan Noise=false) {enabled=false}
   &SolveForcesAndVelocity(Clamp Velocity=false, Limit Acceleration=false)
   &SkeletalMeshLocation(
    Bone / Socket Index=&SelectIntFromArray(Direct Array Index=$Particles.PartitionParticles.Partition, Int Selection Array=DataInterface() {propertyValues {IntData [3]}})
    Bone / Socket Sampling Mode=::NewEnumerator5
    MaskSpawnGroup=false
    Mesh Sampling Type=::NewEnumerator6
    Position Sampling=::NewEnumerator1
    Sampled Position Offset=$Particles.ShapeLocation.ShapePosition
    Skeletal Mesh=$System.Skeletal Mesh
   )
  ]
  rendererClasses [&NiagaraSpriteRendererProperties]
  renderers [{rendererClass=&NiagaraSpriteRendererProperties}]
 }
 Bubbles_Upper_Right {
  bEnabled=false
  emitterUpdateScript [
   &EmitterState()
   &SpawnRate(Spawn Probability=0.5, SpawnRate=0.35, Use Spawn Probability=false)
  ]
  particleSpawnScript [
   &InitializeParticle(
    Color=<rgba 0.5578 0.839 1 1>
    Lifetime Max=3
    Lifetime Min=4
    Lifetime Mode=::NewEnumerator1
    Position Offset=<-0.6 -10 32>
    Sprite Rotation Mode=::NewEnumerator2
    Sprite Size Mode=::NewEnumerator1
    Sprite UV Mode=::NewEnumerator3
    Uniform Sprite Size Max=2
    Uniform Sprite Size Min=4
    UsePositionOffset=true
    Write Material Random=true
   )
   &PartitionParticles(Number Of Partitions=5)
   &ShapeLocation(
    Box Size=<1 8 0>
    MaskSpawnGroup=false
    Override Seed=false
    Rotation Angle=-0.6
    Rotation Mode=::NewEnumerator1
    Shape Primitive=::NewEnumerator2
    Surface Only=false
    Transform Order:ENiagara_TransformOrder=::NewEnumerator1
    UseOwnerScale=true
   )
   &AddVelocity(Velocity=&RandomRangeVector(Minimum=<1 0 1>, Maximum=<-1 0 -1>))
  ]
  particleUpdateScript [
   &ParticleState()
   &ScaleSpriteSize(Scale Factor=&Vector2DFromFloat(Value=&FloatFromCurve(FloatCurve=DataInterface() {propertyValues {Curve {keys [
          {arriveTangent=9.0725, interpMode=RCIM_Cubic, leaveTangent=9.0725, tangentMode=RCTM_User, time=0, value=0}
          {arriveTangent=0.001, interpMode=RCIM_Cubic, leaveTangent=0.001, tangentMode=RCTM_User, time=0.2, value=0.95}
          {arriveTangent=-1.044, interpMode=RCIM_Cubic, leaveTangent=-1.044, tangentMode=RCTM_User, time=0.6, value=0.85}
          {arriveTangent=-2.0004, interpMode=RCIM_Cubic, leaveTangent=-2.0004, tangentMode=RCTM_User, time=1, value=0}
         ]}}})))
   &ScaleColor(
    Scale Alpha=&FloatFromCurve(FloatCurve=DataInterface() {propertyValues {Curve {keys [<at 0 0 RCIM_Cubic>, <at 0.2 1>, {arriveTangent=-1.25, interpMode=RCIM_Cubic, leaveTangent=-1.25, time=0.7, value=1}, <at 1 0 RCIM_Cubic>]}}})
    ScaleA=true
    ScaleRGB=true
   )
   &Drag(Drag=0.5, Use Linear Drag=true, Use Rotational Drag=true) {enabled=false}
   &VortexForce(UseOriginOffset=false, Vortex Axis=<0 1 0>, Vortex Force Amount=1) {enabled=false}
   &CurlNoiseForce(Noise Frequency=20, Noise Strength=3, Pan Noise=false) {enabled=false}
   &SolveForcesAndVelocity(Clamp Velocity=false, Limit Acceleration=false)
   &SkeletalMeshLocation(
    Bone / Socket Index=&SelectIntFromArray(Direct Array Index=$Particles.PartitionParticles.Partition, Int Selection Array=DataInterface() {propertyValues {IntData [3]}})
    Bone / Socket Sampling Mode=::NewEnumerator5
    MaskSpawnGroup=false
    Mesh Sampling Type=::NewEnumerator6
    Position Sampling=::NewEnumerator1
    Sampled Position Offset=$Particles.ShapeLocation.ShapePosition
    Skeletal Mesh=$System.Skeletal Mesh
   )
  ]
  rendererClasses [&NiagaraSpriteRendererProperties]
  renderers [{rendererClass=&NiagaraSpriteRendererProperties}]
 }
 Caustics_BottomLeft {
  bEnabled=false
  emitterUpdateScript [
   &EmitterState(Life Cycle Mode=::NewEnumerator1, Loop Behavior=::NewEnumerator1, Loop Duration Mode=::NewEnumerator1, UseLoopDelay=false)
   &SpawnBurst_Instantaneous(Use Spawn Probability=false, UseLoopCountLimit=false)
   &SpawnRate(Spawn Probability=0.5, SpawnRate=0.2, Use Spawn Probability=false) {enabled=false}
  ]
  particleSpawnScript [
   &InitializeParticle(
    Color=&LerpLinearColors(
     StartColor=<rgba 0.4271 0.7913 1 1>
     EndColor=<rgba 0.0938 0.6699 1 1>
     LerpFactor=&MakeCustomFloatFromBool(Bool=$User.bIsFrontEndPreview, True Float=1, False Float=0)
    )
    Lifetime Max=5.5
    Lifetime Min=5
    Lifetime Mode=::NewEnumerator1
    Position Offset=<-3 5 -11>
    Sprite Rotation Mode=::NewEnumerator2
    Sprite Size Mode=::NewEnumerator1
    Uniform Sprite Size Max=20
    Uniform Sprite Size Min=20
    UsePositionOffset=true
    Write Material Random=true
   )
   &PartitionParticles(Number Of Partitions=5)
   &ShapeLocation(
    Cylinder Height=0
    Cylinder Radius=2
    MaskSpawnGroup=false
    Override Seed=false
    Rotation Mode=::NewEnumerator1
    Shape Primitive=::NewEnumerator1
    Surface Only=false
    Transform Order:ENiagara_TransformOrder=::NewEnumerator1
    UseOwnerScale=true
   )
   &AddVelocity(Velocity=&RandomRangeVector(Minimum=<1.5 0 1.5>, Maximum=<-1.5 0 -1.5>)) {enabled=false}
  ]
  particleUpdateScript [
   &ParticleState(Kill Particles When Lifetime Has Elapsed=false)
   /SetVariables(Particles.SpriteFacing=<0 1 0>) {isSetParameters=true, module=SetVariables_#2}
   &ScaleSpriteSize(Scale Factor=&Vector2DFromFloat(Value=&FloatFromCurve(FloatCurve=DataInterface() {propertyValues {Curve {keys [
          {arriveTangent=9.0725, interpMode=RCIM_Cubic, leaveTangent=9.0725, tangentMode=RCTM_User, time=0, value=0}
          {arriveTangent=0.001, interpMode=RCIM_Cubic, leaveTangent=0.001, tangentMode=RCTM_User, time=0.2, value=0.95}
          {arriveTangent=-1.044, interpMode=RCIM_Cubic, leaveTangent=-1.044, tangentMode=RCTM_User, time=0.6, value=0.85}
          {arriveTangent=-2.0004, interpMode=RCIM_Cubic, leaveTangent=-2.0004, tangentMode=RCTM_User, time=1, value=0}
         ]}}}))) {enabled=false}
   &ScaleColor(
    Scale Alpha=&FloatFromCurve(
     FloatCurve=DataInterface() {propertyValues {Curve {keys [
         {interpMode=RCIM_Cubic, tangentMode=RCTM_User, time=0, value=0.1}
         {interpMode=RCIM_Cubic, tangentMode=RCTM_User, time=0.5, value=0.2}
         {tangentMode=RCTM_User, time=1, value=0.1}
        ]}}}
     Scale Curve=&Lerp_Float(
      A=0.1
      B=0.5
      Alpha=&MakeCustomFloatFromBool(Bool=$User.bIsFrontEndPreview, True Float=1, False Float=0)
     )
    )
    ScaleA=true
    ScaleRGB=true
   )
   &Drag(Drag=0.5, Use Linear Drag=true, Use Rotational Drag=true) {enabled=false}
   &VortexForce(Origin Pull Amount=1, UseOriginOffset=false, Vortex Axis=<0 1 0>, Vortex Force Amount=1) {enabled=false}
   &CurlNoiseForce(Noise Frequency=20, Noise Strength=3, Pan Noise=false) {enabled=false}
   &SolveForcesAndVelocity(Clamp Velocity=false, Limit Acceleration=false)
   &SkeletalMeshLocation(
    Bone / Socket Index=&SelectIntFromArray(Direct Array Index=$Particles.PartitionParticles.Partition, Int Selection Array=DataInterface() {propertyValues {IntData [1]}})
    Bone / Socket Sampling Mode=::NewEnumerator5
    MaskSpawnGroup=false
    Mesh Sampling Type=::NewEnumerator6
    Position Sampling=::NewEnumerator1
    Sampled Position Offset=$Particles.ShapeLocation.ShapePosition
    Skeletal Mesh=$System.Skeletal Mesh
   )
   &CameraOffset() {enabled=false}
  ]
  rendererClasses [&NiagaraSpriteRendererProperties]
  renderers [{rendererClass=&NiagaraSpriteRendererProperties}]
 }
 Caustics_BottomRight {
  bEnabled=false
  emitterUpdateScript [
   &EmitterState(Life Cycle Mode=::NewEnumerator1, Loop Behavior=::NewEnumerator1, Loop Duration Mode=::NewEnumerator1, UseLoopDelay=false)
   &SpawnBurst_Instantaneous(Use Spawn Probability=false, UseLoopCountLimit=false)
   &SpawnRate(Spawn Probability=0.5, SpawnRate=0.2, Use Spawn Probability=false) {enabled=false}
  ]
  particleSpawnScript [
   &InitializeParticle(
    Color=&LerpLinearColors(
     StartColor=<rgba 0.4271 0.7913 1 1>
     EndColor=<rgba 0.0938 0.6699 1 1>
     LerpFactor=&MakeCustomFloatFromBool(Bool=$User.bIsFrontEndPreview, True Float=1, False Float=0)
    )
    Lifetime Max=5
    Lifetime Min=6.5
    Lifetime Mode=::NewEnumerator1
    Position Offset=<-2 -4 -12>
    Sprite Rotation Mode=::NewEnumerator2
    Sprite Size Mode=::NewEnumerator1
    Uniform Sprite Size Max=20
    Uniform Sprite Size Min=20
    UsePositionOffset=true
    Write Material Random=true
   )
   &PartitionParticles(Number Of Partitions=5)
   &ShapeLocation(
    Cylinder Height=0
    Cylinder Radius=2
    MaskSpawnGroup=false
    Override Seed=false
    Rotation Mode=::NewEnumerator1
    Shape Primitive=::NewEnumerator1
    Surface Only=false
    Transform Order:ENiagara_TransformOrder=::NewEnumerator1
    UseOwnerScale=true
   )
   &AddVelocity(Velocity=&RandomRangeVector(Minimum=<1.5 0 1.5>, Maximum=<-1.5 0 -1.5>)) {enabled=false}
  ]
  particleUpdateScript [
   &ParticleState(Kill Particles When Lifetime Has Elapsed=false)
   /SetVariables(Particles.SpriteFacing=<0 1 0>) {isSetParameters=true, module=SetVariables_#2}
   &ScaleSpriteSize(Scale Factor=&Vector2DFromFloat(Value=&FloatFromCurve(FloatCurve=DataInterface() {propertyValues {Curve {keys [
          {arriveTangent=9.0725, interpMode=RCIM_Cubic, leaveTangent=9.0725, tangentMode=RCTM_User, time=0, value=0}
          {arriveTangent=0.001, interpMode=RCIM_Cubic, leaveTangent=0.001, tangentMode=RCTM_User, time=0.2, value=0.95}
          {arriveTangent=-1.044, interpMode=RCIM_Cubic, leaveTangent=-1.044, tangentMode=RCTM_User, time=0.6, value=0.85}
          {arriveTangent=-2.0004, interpMode=RCIM_Cubic, leaveTangent=-2.0004, tangentMode=RCTM_User, time=1, value=0}
         ]}}}))) {enabled=false}
   &ScaleColor(
    Scale Alpha=&FloatFromCurve(
     FloatCurve=DataInterface() {propertyValues {Curve {keys [
         {interpMode=RCIM_Cubic, tangentMode=RCTM_User, time=0, value=0.1}
         {interpMode=RCIM_Cubic, tangentMode=RCTM_User, time=0.5, value=0.2}
         {tangentMode=RCTM_User, time=1, value=0.1}
        ]}}}
     Scale Curve=&Lerp_Float(
      A=0.1
      B=0.5
      Alpha=&MakeCustomFloatFromBool(Bool=$User.bIsFrontEndPreview, True Float=1, False Float=0)
     )
    )
    ScaleA=true
    ScaleRGB=true
   )
   &Drag(Drag=0.5, Use Linear Drag=true, Use Rotational Drag=true) {enabled=false}
   &VortexForce(Origin Pull Amount=1, UseOriginOffset=false, Vortex Axis=<0 1 0>, Vortex Force Amount=1) {enabled=false}
   &CurlNoiseForce(Noise Frequency=20, Noise Strength=3, Pan Noise=false) {enabled=false}
   &SolveForcesAndVelocity(Clamp Velocity=false, Limit Acceleration=false)
   &SkeletalMeshLocation(
    Bone / Socket Index=&SelectIntFromArray(Direct Array Index=$Particles.PartitionParticles.Partition, Int Selection Array=DataInterface() {propertyValues {IntData [1]}})
    Bone / Socket Sampling Mode=::NewEnumerator5
    MaskSpawnGroup=false
    Mesh Sampling Type=::NewEnumerator6
    Position Sampling=::NewEnumerator1
    Sampled Position Offset=$Particles.ShapeLocation.ShapePosition
    Skeletal Mesh=$System.Skeletal Mesh
   )
   &CameraOffset() {enabled=false}
  ]
  rendererClasses [&NiagaraSpriteRendererProperties]
  renderers [{rendererClass=&NiagaraSpriteRendererProperties}]
 }
 Caustics_New {
  emitterUpdateScript [
   &EmitterState(
    Inactive Response=::NewEnumerator1
    Life Cycle Mode=::NewEnumerator1
    Loop Behavior=::NewEnumerator1
    Loop Duration Mode=::NewEnumerator1
    UseLoopDelay=false
   )
   &SpawnBurst_Instantaneous(Spawn Count=5, Use Spawn Probability=false, UseLoopCountLimit=false) {enabled=false}
   &SpawnRate(Spawn Probability=0.5, SpawnRate=3, Use Spawn Probability=false)
  ]
  particleSpawnScript [
   &PartitionParticles(Number Of Partitions=5)
   &InitializeParticle(
    Color=&LerpLinearColors(
     StartColor=<rgba 0.4271 0.7913 1 1>
     EndColor=<rgba 0.0938 0.6699 1 1>
     LerpFactor=&MakeCustomFloatFromBool(Bool=$User.bIsFrontEndPreview, True Float=1, False Float=0)
    )
    Lifetime Max=4
    Lifetime Min=3
    Lifetime Mode=::NewEnumerator1
    Position Offset=&SelectVectorFromArray(
     Direct Array Index=$Particles.PartitionParticles.Partition
     Vector Selection Array=DataInterface() {propertyValues {FloatData [<-2 -2 -7>, <2 -2 0>, <-5 -2 8>, <7 -2 8>, <7 -2 -7>]}}
    )
    Position Offset Coordinate Space=::Local
    Sprite Rotation Mode=::NewEnumerator2
    Sprite Size Mode=::NewEnumerator1
    Uniform Sprite Size Max=15
    Uniform Sprite Size Min=14
    UsePositionOffset=true
    Write Material Random=true
   )
   &ShapeLocation(
    Cylinder Height=0
    Cylinder Radius=2
    MaskSpawnGroup=false
    Offset Coordinate Space=::Local
    Offset Mode=::NewEnumerator0
    Override Seed=false
    Rotation Mode=::NewEnumerator1
    Shape Origin=$Particles.Position
    Shape Primitive=::NewEnumerator1
    Surface Only=false
    Transform Order:ENiagara_TransformOrder=::NewEnumerator1
    UseOwnerScale=true
   )
   &PartitionParticles(Number Of Partitions=3) {enabled=false, module=PartitionParticles001}
  ]
  particleUpdateScript [
   &ParticleState()
   /SetVariables(Particles.SpriteFacing=<0 1 0>) {isSetParameters=true, module=SetVariables_#2}
   &ScaleSpriteSize(Scale Factor=&Vector2DFromFloat(Value=&FloatFromCurve(FloatCurve=DataInterface() {propertyValues {Curve {keys [
          {arriveTangent=9.0725, interpMode=RCIM_Cubic, leaveTangent=9.0725, tangentMode=RCTM_User, time=0, value=0}
          {arriveTangent=0.001, interpMode=RCIM_Cubic, leaveTangent=0.001, tangentMode=RCTM_User, time=0.2, value=0.95}
          {arriveTangent=-1.044, interpMode=RCIM_Cubic, leaveTangent=-1.044, tangentMode=RCTM_User, time=0.6, value=0.85}
          {arriveTangent=-2.0004, interpMode=RCIM_Cubic, leaveTangent=-2.0004, tangentMode=RCTM_User, time=1, value=0}
         ]}}}))) {enabled=false}
   &ScaleColor(
    Scale Alpha=&FloatFromCurve(
     FloatCurve=DataInterface() {propertyValues {Curve {keys [
         {interpMode=RCIM_Cubic, tangentMode=RCTM_User, time=0, value=0}
         {interpMode=RCIM_Cubic, tangentMode=RCTM_User, time=0.5, value=0.2}
         {tangentMode=RCTM_User, time=1, value=0}
        ]}}}
     Scale Curve=&Lerp_Float(
      A=0.1
      B=0.5
      Alpha=&MakeCustomFloatFromBool(Bool=$User.bIsFrontEndPreview, True Float=1, False Float=0)
     )
    )
    ScaleA=true
    ScaleRGB=true
   )
   &CameraOffset() {enabled=false}
   /SetVariables(Particles.Position=&Lerp_Vector(
     A=Linked() {linkedVariable {name=Particles.Position, type=NiagaraPosition}}
     B=&AddVectorToPosition(
      Position=$Particles.Position
      Vector=&RandomRangeVector(Minimum=<-0.05 0 -0.05>, Maximum=<0.05 0 0.05>)
     )
     Alpha=&FloatFromCurve(FloatCurve=DataInterface() {propertyValues {Curve {keys [<at 0 0 RCIM_Cubic>, {tangentMode=RCTM_User, time=1, value=1}]}}})
    )) {isSetParameters=true}
  ]
  rendererClasses [&NiagaraSpriteRendererProperties]
  renderers [{rendererClass=&NiagaraSpriteRendererProperties}]
 }
 Caustics_TopLeft {
  bEnabled=false
  emitterUpdateScript [
   &EmitterState(Life Cycle Mode=::NewEnumerator1, Loop Behavior=::NewEnumerator1, Loop Duration Mode=::NewEnumerator1, UseLoopDelay=false)
   &SpawnBurst_Instantaneous(Use Spawn Probability=false, UseLoopCountLimit=false)
   &SpawnRate(Spawn Probability=0.5, SpawnRate=0.2, Use Spawn Probability=false) {enabled=false}
  ]
  particleSpawnScript [
   &InitializeParticle(
    Color=&LerpLinearColors(
     StartColor=<rgba 0.4271 0.7913 1 1>
     EndColor=<rgba 0.0938 0.6699 1 1>
     LerpFactor=&MakeCustomFloatFromBool(Bool=$User.bIsFrontEndPreview, True Float=1, False Float=0)
    )
    Lifetime Max=5.5
    Lifetime Min=4
    Lifetime Mode=::NewEnumerator1
    Position Offset=<-2 3 6>
    Sprite Rotation Mode=::NewEnumerator2
    Sprite Size Mode=::NewEnumerator1
    Uniform Sprite Size Max=20
    Uniform Sprite Size Min=20
    UsePositionOffset=true
    Write Material Random=true
   )
   &PartitionParticles(Number Of Partitions=5)
   &ShapeLocation(
    Cylinder Height=0
    Cylinder Radius=2
    MaskSpawnGroup=false
    Override Seed=false
    Rotation Mode=::NewEnumerator1
    Shape Primitive=::NewEnumerator1
    Surface Only=false
    Transform Order:ENiagara_TransformOrder=::NewEnumerator1
    UseOwnerScale=true
   )
   &AddVelocity(Velocity=&RandomRangeVector(Minimum=<1.5 0 1.5>, Maximum=<-1.5 0 -1.5>)) {enabled=false}
  ]
  particleUpdateScript [
   &ParticleState(Kill Particles When Lifetime Has Elapsed=false)
   /SetVariables(Particles.SpriteFacing=<0 1 0>) {isSetParameters=true, module=SetVariables_#2}
   &ScaleSpriteSize(Scale Factor=&Vector2DFromFloat(Value=&FloatFromCurve(FloatCurve=DataInterface() {propertyValues {Curve {keys [
          {arriveTangent=9.0725, interpMode=RCIM_Cubic, leaveTangent=9.0725, tangentMode=RCTM_User, time=0, value=0}
          {arriveTangent=0.001, interpMode=RCIM_Cubic, leaveTangent=0.001, tangentMode=RCTM_User, time=0.2, value=0.95}
          {arriveTangent=-1.044, interpMode=RCIM_Cubic, leaveTangent=-1.044, tangentMode=RCTM_User, time=0.6, value=0.85}
          {arriveTangent=-2.0004, interpMode=RCIM_Cubic, leaveTangent=-2.0004, tangentMode=RCTM_User, time=1, value=0}
         ]}}}))) {enabled=false}
   &ScaleColor(
    Scale Alpha=&FloatFromCurve(
     FloatCurve=DataInterface() {propertyValues {Curve {keys [
         {interpMode=RCIM_Cubic, tangentMode=RCTM_User, time=0, value=0.1}
         {interpMode=RCIM_Cubic, tangentMode=RCTM_User, time=0.5, value=0.2}
         {tangentMode=RCTM_User, time=1, value=0.1}
        ]}}}
     Scale Curve=&Lerp_Float(
      A=0.1
      B=0.5
      Alpha=&MakeCustomFloatFromBool(Bool=$User.bIsFrontEndPreview, True Float=1, False Float=0)
     )
    )
    ScaleA=true
    ScaleRGB=true
   )
   &Drag(Drag=0.5, Use Linear Drag=true, Use Rotational Drag=true) {enabled=false}
   &VortexForce(Origin Pull Amount=1, UseOriginOffset=false, Vortex Axis=<0 1 0>, Vortex Force Amount=1) {enabled=false}
   &CurlNoiseForce(Noise Frequency=20, Noise Strength=3, Pan Noise=false) {enabled=false}
   &SolveForcesAndVelocity(Clamp Velocity=false, Limit Acceleration=false)
   &SkeletalMeshLocation(
    Bone / Socket Index=&SelectIntFromArray(Direct Array Index=$Particles.PartitionParticles.Partition, Int Selection Array=DataInterface() {propertyValues {IntData [1]}})
    Bone / Socket Sampling Mode=::NewEnumerator5
    MaskSpawnGroup=false
    Mesh Sampling Type=::NewEnumerator6
    Position Sampling=::NewEnumerator1
    Sampled Position Offset=$Particles.ShapeLocation.ShapePosition
    Skeletal Mesh=$System.Skeletal Mesh
   )
   &CameraOffset() {enabled=false}
  ]
  rendererClasses [&NiagaraSpriteRendererProperties]
  renderers [{rendererClass=&NiagaraSpriteRendererProperties}]
 }
 Caustics_TopLeft001 {
  bEnabled=false
  emitterUpdateScript [
   &EmitterState(Life Cycle Mode=::NewEnumerator1, Loop Behavior=::NewEnumerator1, Loop Duration Mode=::NewEnumerator1, UseLoopDelay=false)
   &SpawnBurst_Instantaneous(Use Spawn Probability=false, UseLoopCountLimit=false)
   &SpawnRate(Spawn Probability=0.5, SpawnRate=100, Use Spawn Probability=false) {enabled=false}
  ]
  particleSpawnScript [
   &InitializeParticle(
    Color=&LerpLinearColors(
     StartColor=<rgba 0.4271 0.7913 1 1>
     EndColor=<rgba 0.0938 0.6699 1 1>
     LerpFactor=&MakeCustomFloatFromBool(Bool=$User.bIsFrontEndPreview, True Float=1, False Float=0)
    )
    Lifetime Max=5.5
    Lifetime Min=4
    Lifetime Mode=::NewEnumerator1
    Position Offset=<-2 -4 5>
    Sprite Rotation Mode=::NewEnumerator2
    Sprite Size Mode=::NewEnumerator1
    Uniform Sprite Size Max=17.5
    Uniform Sprite Size Min=17.5
    UsePositionOffset=true
    Write Material Random=true
   )
   &PartitionParticles(Number Of Partitions=5)
   &ShapeLocation(
    Cylinder Height=0
    Cylinder Radius=2
    MaskSpawnGroup=false
    Override Seed=false
    Rotation Mode=::NewEnumerator1
    Shape Primitive=::NewEnumerator1
    Surface Only=false
    Transform Order:ENiagara_TransformOrder=::NewEnumerator1
    UseOwnerScale=true
   )
   &AddVelocity(Velocity=&RandomRangeVector(Minimum=<1.5 0 1.5>, Maximum=<-1.5 0 -1.5>)) {enabled=false}
  ]
  particleUpdateScript [
   &ParticleState(Kill Particles When Lifetime Has Elapsed=false)
   /SetVariables(Particles.SpriteFacing=<0 1 0>) {isSetParameters=true, module=SetVariables_#2}
   &ScaleSpriteSize(Scale Factor=&Vector2DFromFloat(Value=&FloatFromCurve(FloatCurve=DataInterface() {propertyValues {Curve {keys [
          {arriveTangent=9.0725, interpMode=RCIM_Cubic, leaveTangent=9.0725, tangentMode=RCTM_User, time=0, value=0}
          {arriveTangent=0.001, interpMode=RCIM_Cubic, leaveTangent=0.001, tangentMode=RCTM_User, time=0.2, value=0.95}
          {arriveTangent=-1.044, interpMode=RCIM_Cubic, leaveTangent=-1.044, tangentMode=RCTM_User, time=0.6, value=0.85}
          {arriveTangent=-2.0004, interpMode=RCIM_Cubic, leaveTangent=-2.0004, tangentMode=RCTM_User, time=1, value=0}
         ]}}}))) {enabled=false}
   &ScaleColor(
    Scale Alpha=&FloatFromCurve(
     FloatCurve=DataInterface() {propertyValues {Curve {keys [
         {interpMode=RCIM_Cubic, tangentMode=RCTM_User, time=0, value=0.1}
         {interpMode=RCIM_Cubic, tangentMode=RCTM_User, time=0.5, value=0.2}
         {tangentMode=RCTM_User, time=1, value=0.1}
        ]}}}
     Scale Curve=&Lerp_Float(
      A=0.1
      B=0.5
      Alpha=&MakeCustomFloatFromBool(Bool=$User.bIsFrontEndPreview, True Float=1, False Float=0)
     )
    )
    ScaleA=true
    ScaleRGB=true
   )
   &Drag(Drag=0.5, Use Linear Drag=true, Use Rotational Drag=true) {enabled=false}
   &VortexForce(Origin Pull Amount=1, UseOriginOffset=false, Vortex Axis=<0 1 0>, Vortex Force Amount=1) {enabled=false}
   &CurlNoiseForce(Noise Frequency=20, Noise Strength=3, Pan Noise=false) {enabled=false}
   &SolveForcesAndVelocity(Clamp Velocity=false, Limit Acceleration=false)
   &SkeletalMeshLocation(
    Bone / Socket Index=&SelectIntFromArray(Direct Array Index=$Particles.PartitionParticles.Partition, Int Selection Array=DataInterface() {propertyValues {IntData [1]}})
    Bone / Socket Sampling Mode=::NewEnumerator5
    MaskSpawnGroup=false
    Mesh Sampling Type=::NewEnumerator6
    Position Sampling=::NewEnumerator1
    Sampled Position Offset=$Particles.ShapeLocation.ShapePosition
    Skeletal Mesh=$System.Skeletal Mesh
   )
   &CameraOffset() {enabled=false}
  ]
  rendererClasses [&NiagaraSpriteRendererProperties]
  renderers [{rendererClass=&NiagaraSpriteRendererProperties}]
 }
 Fish {
  bEnabled=false
  emitterUpdateScript [
   &EmitterState()
   &SpawnPerUnit(
    Max Movement Threshold=1
    Module.Use Max Movement Threshold=false
    Module.Use Movement Tolerance=false
    Movement Tolerance=0
    Spawn Spacing=1
    Use Spawn Probability=false
   ) {enabled=false}
   &SpawnRate(SpawnRate=4, Use Spawn Probability=false)
  ]
  particleSpawnScript [
   &InitializeParticle(
    Color=<rgba 0.5578 0.839 1 1>
    Lifetime Mode=::NewEnumerator1
    Sprite Rotation Mode=::NewEnumerator2
    Sprite Size Mode=::NewEnumerator1
    Sprite UV Mode=::NewEnumerator3
    Uniform Sprite Size Max=1
    Uniform Sprite Size Min=2
    UsePositionOffset=false
    Write Material Random=true
   )
   &PartitionParticles(Number Of Partitions=6)
   &ShapeLocation(
    Box Size=<0 0 0>
    MaskSpawnGroup=false
    Override Seed=false
    Rotation Mode=::NewEnumerator1
    Shape Primitive=::NewEnumerator2
    Surface Only=false
    Transform Order:ENiagara_TransformOrder=::NewEnumerator1
    UseOwnerScale=true
   )
   &AddVelocity(Velocity=&RandomRangeVector(Minimum=<2 0 2>, Maximum=<-2 0 -2>))
  ]
  particleUpdateScript [
   &ParticleState()
   &ScaleSpriteSize(Scale Factor=&Vector2DFromFloat(Value=&FloatFromCurve(FloatCurve=DataInterface() {propertyValues {Curve {keys [
          {arriveTangent=9.0725, interpMode=RCIM_Cubic, leaveTangent=9.0725, tangentMode=RCTM_User, time=0, value=0}
          {arriveTangent=0.001, interpMode=RCIM_Cubic, leaveTangent=0.001, tangentMode=RCTM_User, time=0.2, value=0.95}
          {arriveTangent=-1.044, interpMode=RCIM_Cubic, leaveTangent=-1.044, tangentMode=RCTM_User, time=0.6, value=0.85}
          {arriveTangent=-2.0004, interpMode=RCIM_Cubic, leaveTangent=-2.0004, tangentMode=RCTM_User, time=1, value=0}
         ]}}})))
   &ScaleColor(
    Scale Alpha=&FloatFromCurve(FloatCurve=DataInterface() {propertyValues {Curve {keys [<at 0 0 RCIM_Cubic>, <at 0.2 1>, {arriveTangent=-1.25, interpMode=RCIM_Cubic, leaveTangent=-1.25, time=0.7, value=1}, <at 1 0 RCIM_Cubic>]}}})
    ScaleA=true
    ScaleRGB=true
   )
   &Drag(Drag=0.2, Use Linear Drag=true, Use Rotational Drag=true) {enabled=false}
   &VortexForce(UseOriginOffset=false, Vortex Axis=<0 0 0>, Vortex Force Amount=8) {enabled=false}
   &CurlNoiseForce(Noise Frequency=20, Noise Strength=3, Pan Noise=false) {enabled=false}
   &CameraOffset(Camera Offset Amount=-3)
   &SkeletalMeshLocation(
    Bone / Socket Index=&SelectIntFromArray(Direct Array Index=$Particles.PartitionParticles.Partition, Int Selection Array=DataInterface() {propertyValues {IntData [4]}})
    Bone / Socket Sampling Mode=::NewEnumerator5
    MaskSpawnGroup=false
    Mesh Sampling Type=::NewEnumerator6
    Orientation Sampling=::NewEnumerator1
    Position Sampling=::NewEnumerator1
    Sampled Position Offset=$Particles.ShapeLocation.ShapePosition
    Sampled Position Offset Coordinate Space:ENiagaraSkelMeshTransforms=::NewEnumerator2
    Skeletal Mesh=$System.Skeletal Mesh
    Source Movement Scale:Vector3f=<0 0 0>
   )
   &SolveForcesAndVelocity(Clamp Velocity=false, Limit Acceleration=false)
  ]
  rendererClasses [&NiagaraSpriteRendererProperties]
  renderers [{rendererClass=&NiagaraSpriteRendererProperties}]
 }
 StarFish {
  bEnabled=false
  emitterUpdateScript [
   &EmitterState()
   &SpawnRate(SpawnRate=2, Use Spawn Probability=false)
  ]
  particleSpawnScript [
   &InitializeParticle(
    Color=<rgba 0.5578 0.839 1 1>
    Lifetime Min=2
    Lifetime Mode=::NewEnumerator1
    Sprite Rotation Mode=::NewEnumerator2
    Sprite Size Mode=::NewEnumerator1
    Sprite UV Mode=::NewEnumerator3
    Uniform Sprite Size Max=1
    Uniform Sprite Size Min=2
    UsePositionOffset=false
    Write Material Random=true
   )
   &PartitionParticles(Number Of Partitions=6)
   &ShapeLocation(
    Cylinder Height=0
    Cylinder Radius=4
    MaskSpawnGroup=false
    Override Seed=false
    Rotation Axis=<0 1 0>
    Rotation Mode=::NewEnumerator1
    Shape Primitive=::NewEnumerator1
    Surface Only=true
    Transform Order:ENiagara_TransformOrder=::NewEnumerator1
    Use Endcaps In Surface Only Mode=false
    UseOwnerScale=true
   )
   &AddVelocity(Velocity=&RandomRangeVector(Minimum=<3 0 3>, Maximum=<-3 0 -3>)) {enabled=false}
  ]
  particleUpdateScript [
   &ParticleState()
   &ScaleSpriteSize(Scale Factor=&Vector2DFromFloat(Value=&FloatFromCurve(FloatCurve=DataInterface() {propertyValues {Curve {keys [
          {arriveTangent=9.0725, interpMode=RCIM_Cubic, leaveTangent=9.0725, tangentMode=RCTM_User, time=0, value=0}
          {arriveTangent=0.001, interpMode=RCIM_Cubic, leaveTangent=0.001, tangentMode=RCTM_User, time=0.2, value=0.95}
          {arriveTangent=-1.044, interpMode=RCIM_Cubic, leaveTangent=-1.044, tangentMode=RCTM_User, time=0.6, value=0.85}
          {arriveTangent=-2.0004, interpMode=RCIM_Cubic, leaveTangent=-2.0004, tangentMode=RCTM_User, time=1, value=0}
         ]}}})))
   &ScaleColor(
    Scale Alpha=&FloatFromCurve(FloatCurve=DataInterface() {propertyValues {Curve {keys [<at 0 0 RCIM_Cubic>, <at 0.2 1>, {arriveTangent=-1.25, interpMode=RCIM_Cubic, leaveTangent=-1.25, time=0.7, value=1}, <at 1 0 RCIM_Cubic>]}}})
    ScaleA=true
    ScaleRGB=true
   )
   &Drag(Drag=0.5, Use Linear Drag=true, Use Rotational Drag=true)
   &VortexForce(UseOriginOffset=false, Vortex Axis=<1 0 0>, Vortex Force Amount=10) {enabled=false}
   &CurlNoiseForce(Noise Frequency=20, Noise Strength=3, Pan Noise=false) {enabled=false}
   &SkeletalMeshLocation(
    Bone / Socket Index=&SelectIntFromArray(Direct Array Index=$Particles.PartitionParticles.Partition, Int Selection Array=DataInterface() {propertyValues {IntData [5]}})
    Bone / Socket Sampling Mode=::NewEnumerator5
    Inherited Velocity Speed Limit=0
    MaskSpawnGroup=false
    Mesh Sampling Type=::NewEnumerator6
    Position Sampling=::NewEnumerator1
    Sampled Position Offset=$Particles.ShapeLocation.ShapePosition
    Sampled Position Offset Coordinate Space:ENiagaraSkelMeshTransforms=::NewEnumerator2
    Skeletal Mesh=$System.Skeletal Mesh
    Velocity Sampling=::NewEnumerator0
    VelocitySpeedLimit=true
   )
   &SolveForcesAndVelocity(Clamp Velocity=false, Limit Acceleration=false)
   &CameraOffset(Camera Offset Amount=-3)
  ]
  rendererClasses [&NiagaraSpriteRendererProperties]
  renderers [{rendererClass=&NiagaraSpriteRendererProperties}]
 }
}
summary {
 systemName=NS_Guitar_Vacation
 emitters [
  {bEnabled=false, emitterName=Bubbles_Main, rendererClasses [&NiagaraSpriteRendererProperties]}
  {bEnabled=false, emitterName=Bubbles_Bottom_Left, rendererClasses [&NiagaraSpriteRendererProperties]}
  {bEnabled=false, emitterName=Bubbles_Upper_Left, rendererClasses [&NiagaraSpriteRendererProperties]}
  {bEnabled=false, emitterName=Bubbles_Upper_Right, rendererClasses [&NiagaraSpriteRendererProperties]}
  {bEnabled=false, emitterName=StarFish, rendererClasses [&NiagaraSpriteRendererProperties]}
  {bEnabled=false, emitterName=Fish, rendererClasses [&NiagaraSpriteRendererProperties]}
  {bEnabled=false, emitterName=Caustics_BottomLeft, rendererClasses [&NiagaraSpriteRendererProperties]}
  {bEnabled=false, emitterName=Caustics_BottomRight, rendererClasses [&NiagaraSpriteRendererProperties]}
  {bEnabled=false, emitterName=Caustics_TopLeft, rendererClasses [&NiagaraSpriteRendererProperties]}
  {bEnabled=false, emitterName=Caustics_TopLeft001, rendererClasses [&NiagaraSpriteRendererProperties]}
  {emitterName=Bubbles_Main_New, rendererClasses [&NiagaraSpriteRendererProperties]}
  {emitterName=Caustics_New, rendererClasses [&NiagaraSpriteRendererProperties]}
 ]
 userVariables [User.bIsFrontEndPreview {type=NiagaraBool}]
}
