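# Animation Blueprint for the Vacation guitar cosmetic in SparksCosmetics, applying additive idle and fidget motion over the input pose.
# A state machine loops the idle, enters a one-shot fidget on IsFidgeting and returns automatically, using long crossfades.
# A Do Once and random Delay chain is built to pulse IsFidgeting, but the update event is not wired to it.
# TimeBetweenFidgets is unused.
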AnimBlueprint /SparksCosmetics/Instrument/Guitar/Vacation/Meshes/Vacation_Guitar_AnimBP
target_skeleton=&/Vacation_Guitar_Skeleton
graphs [
 AnimGraph {nodes [
   AnimGraphNode_StateMachine_0 = Cosmetic Motion() {is=title AnimGraphNode_StateMachine}
   Output Pose(Result=Apply Additive(Base=Input Pose(), Additive=@AnimGraphNode_StateMachine_0) {alphaBoolBlend {blendOption=Linear}})
  ]}
 Cosmetic Motion {
  entry=Idle_Loop
  path=AnimGraph.AnimGraphNode_StateMachine_0.Cosmetic Motion
  ids=id
  nodes [
   AnimStateNode_0 = Idle_Loop() {is=title AnimStateNode}
   AnimStateNode_1 = Fidget_A() {is=title AnimStateNode}
   AnimStateTransitionNode_0 = AnimStateTransitionNode() {crossfadeDuration=8.8}
   AnimStateTransitionNode_1 = AnimStateTransitionNode() {bAutomaticRuleBasedOnSequencePlayerInState=true, crossfadeDuration=8.8}
  ]
  links [
   {
    dst=Idle_Loop
    rule=AnimGraph.AnimGraphNode_StateMachine_0.Cosmetic Motion.AnimStateTransitionNode_1.Transition
    src=Fidget_A
    via=AnimStateTransitionNode_1
   }
   {
    dst=Fidget_A
    rule=AnimGraph.AnimGraphNode_StateMachine_0.Cosmetic Motion.AnimStateTransitionNode_0.Transition
    src=Idle_Loop
    via=AnimStateTransitionNode_0
   }
  ]
 }
 Idle_Loop {
  path=AnimGraph.AnimGraphNode_StateMachine_0.Cosmetic Motion.AnimStateNode_0.Idle_Loop
  nodes [Output Animation Pose(Result="Sequence Player (additive) 'Vacation_Guitar_Idle_Loop'"!(bLoopAnimation:bool=True) -> (Pose) {sequence=&/Vacation_Guitar_Idle_Loop})]
 }
 Fidget_A {
  path=AnimGraph.AnimGraphNode_StateMachine_0.Cosmetic Motion.AnimStateNode_1.Fidget_A
  nodes [Output Animation Pose(Result="Sequence Player (additive) 'Vacation_Guitar_Fidget_A'"!(bLoopAnimation:bool=False) -> (Pose) {bLoopAnimation=false, sequence=&/Vacation_Guitar_Fidget_A})]
 }
 Transition {
  path=AnimGraph.AnimGraphNode_StateMachine_0.Cosmetic Motion.AnimStateTransitionNode_0.Transition
  nodes [Result(bCanEnterTransition=Get IsFidgeting() -> (IsFidgeting:bool))]
 }
 Transition {path=AnimGraph.AnimGraphNode_StateMachine_0.Cosmetic Motion.AnimStateTransitionNode_1.Transition, nodes [Result(bCanEnterTransition=true)]}
 EventGraph {nodes [
   Event BlueprintUpdateAnimation()
   n1 = RandomFloatInRange(Min=8, Max=12)
   Delay(execute=@n3.Completed, Duration=@n1)
   > Set IsFidgeting(IsFidgeting:bool=true) -> (Output_Get:bool)
   > Delay(Duration=0.2)
   > n2 = Set IsFidgeting(IsFidgeting=false)
   n3 = Do Once(Reset=@n2.then)
  ]}
]
variables [
 IsFidgeting {type=boolean}
 TimeBetweenFidgets {type=number, default=10}
]
